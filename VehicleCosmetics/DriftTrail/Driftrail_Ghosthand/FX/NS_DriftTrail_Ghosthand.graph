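# Niagara system for the Ghosthand drift trail vehicle cosmetic, with per-emitter color overrides for painted variants.
# A ribbon emitter spawns two backward-moving trails that inherit car movement, narrowing and fading over their lifetime.
# A sprite emitter layers four spawn groups at skeletal wheel bones, switching size, lifetime, velocity and jitter per group, with VisibilityTag set from the group.
# All emitters use per-wheel tire FX masking; the wheel-wrap mesh emitter is disabled.
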NiagaraSystem /VehicleCosmetics/DriftTrail/Driftrail_Ghosthand/FX/NS_DriftTrail_Ghosthand
dependencies {
 usedDataInterfaces [
  &NiagaraDataInterfaceCurve
  &NiagaraDataInterfaceSpriteRendererInfo
  &NiagaraDataInterfaceSkeletalMesh
 ]
 usedDynamicInputs [&RandomRangeVector2D, &Multiply_Vector2D_ByFloat]
 usedModules [
  &self:SystemSpawnScript.SystemScriptSource.SystemScriptGraph.NiagaraNodeAssignment_0.SetVariables_#1
  &FN_UnpackVehicleFlags
  /VehicleCosmetics/FX/Niagara/Modules/VC_SetVehicleSpeedAttribs
  &SystemState
 ]
}
emitters {
 NE_VC_DriftTrail_Ribbon_Parent {
  emitterUpdateScript [
   &EmitterState()
   &VC_RateQualityHandler(Default Rate=30)
   &VC_SpawnRibbon_v1(Number of Ribbons=2, SpawnRate=$Emitter.Spawn Rate)
  ]
  particleSpawnScript [
   &VC_PaintedVariantsOverride(
    Black=<rgba 0 0 0 5>
    BurntSienna=<rgba 0.0573 0.015 0.009 3>
    Cobalt=<rgba 0 0.127 1 10>
    Crimson=<rgba 0.4635 0 0.0126 5>
    DefaultColor=<rgba 0.4542 2 0.9413 5>
    ForestGreen=<rgba 0 0.276 0.0139 5>
    Grey=<rgba 0.1514 0.1514 0.1615 5>
    Lime=<rgba 0.4555 0.776 0 5>
    Orange=<rgba 0.6724 0.1426 0.0038 10>
    Pink=<rgba 0.7347 0.2348 0.7448 3>
    Purple=<rgba 0.1621 0.0606 0.8958 10>
    Saffron=<rgba 0.8854 0.6788 0 5>
    SkyBlue=<rgba 0.0328 0.5462 1 3>
    TitaniumWhite=<rgba 0.5596 0.5596 0.5625 5>
   )
   &VC_InitializeRibbon_v1(
    Color=$Particles.Color
    Custom Offset=true
    Enable Skeletal Position=true
    Lifetime=0.65
    Local Vector Offset=&AddVectorToPosition(Position=<-50 0 -20>, Vector=$User.offset)
    Offset Amount=1
    Ribbon Width=70
   )
   &AddVelocity(Velocity=<-1400 0 0>) {module=AddVelocity001}
  ]
  particleUpdateScript [
   &FN_DisableTireFX(Mask Index Attribute:NiagaraInt32=$Particles.SpawnGroup, Mask Index FL=2, Mask Index FR=3, Mask Index RL=0, Mask Index RR=1)
   &ParticleState()
   &InheritSourceMovement(
    Applied Position Scale=&Multiply_VectorByFloat(
     Vector=<1 1 1>
     Float=&FloatFromCurve(FloatCurve=DataInterface() {propertyValues {Curve {keys [
          {interpMode=RCIM_Cubic, leaveTangent=-0.1311, tangentMode=RCTM_Break, time=0, value=1}
          {arriveTangent=-0.0803, interpMode=RCIM_Cubic, tangentMode=RCTM_Break, time=1, value=0.9}
         ]}}})
    )
    ApplyForce=false
    ApplyPosition=true
    ApplyPositionFalloff=false
   )
   &SolveForcesAndVelocity(Clamp Velocity=false, Limit Acceleration=false)
   &ScaleColor(
    Scale Alpha=&FloatFromCurve(
     FloatCurve=DataInterface() {propertyValues {Curve {keys [
         {interpMode=RCIM_Cubic, leaveTangent=-0.7406, tangentMode=RCTM_Break, time=0.0073, value=0.9921}
         {interpMode=RCIM_Cubic, tangentMode=RCTM_Break, time=0.905, value=-0.0353}
        ]}}}
     Scale Curve=3
    )
    Scale RGB=&VectorFromFloat(Value=1)
    ScaleA=true
    ScaleRGB=true
   )
   &ScaleRibbonWidth(Ribbon Width Scale=&FloatFromCurve(FloatCurve=DataInterface() {propertyValues {Curve {keys [<at 0.0051 1.0118>,<at 0.2 1>,<at 0.5688 0.6941>,<at 0.9974 0.6941>]}}}))
   &VC_UpdateRibbon_v1(Enable Skeletal Position=true)
  ]
  rendererClasses [&NiagaraRibbonRendererProperties]
  renderers [
   {rendererClass=&NiagaraRibbonRendererProperties}
   {rendererClass=&NiagaraRibbonRendererProperties, rendererIndex=1}
  ]
 }
 NE_VC_DriftTrail_Sprite_Parent {
  emitterUpdateScript [
   &EmitterState(Life Cycle Mode=::NewEnumerator1, Loop Duration=3, UseLoopDelay=false)
   &SpawnBurst_Instantaneous(Age:NiagaraFloat=$Emitter.NormalizedLoopAge, Spawn Count=2, Spawn Group=3, Use Spawn Probability=true, UseLoopCountLimit=false)
   &SpawnRate(SpawnGroup=2, SpawnRate=35, Use Spawn Probability=false) {module=SpawnRate002}
   &SpawnRate(SpawnGroup=1, SpawnRate=100, Use Spawn Probability=false) {module=SpawnRate001}
   &SpawnRate(SpawnRate=30, Use Spawn Probability=false)
  ]
  particleSpawnScript [
   &VC_PaintedVariantsOverride(
    Black=<rgba 0 0 0 3>
    BurntSienna=<rgba 0.0642 0.0198 0.0079 2>
    Cobalt=<rgba 0 0.0185 1 2>
    DefaultColor=<rgba 0.6843 2 1.3042 1>
    Orange=<rgba 0.6724 0.2346 0 2>
    Pink=<rgba 1 0.3073 0.7481 1>
    SkyBlue=<rgba 0 0.6654 1 1>
    TitaniumWhite=<rgba 1 1 1 1>
   )
   &InitializeParticle(
    Color=$Particles.Color
    Lifetime=&MergeHelper_FloatSwitch(
     0=&RandomRangeFloat(Minimum=0, Maximum=0.1)
     1=&RandomRangeFloat(Minimum=0.15, Maximum=0.3)
     2=0.3
     3=3
     4=0
     Index=&SpawnGroup()
     Num=::NewEnumerator3
    )
    Mesh Scale=&Multiply_VectorByFloat(Vector=<0.6 0.7 0.6>, Float=3)
    Mesh Scale Mode=::NewEnumerator4
    Position Offset=&AddVectorToPosition(
     Position=$User.DTCarOffset
     Vector=&MergeHelper_VectorSwitch(
      0=<-250 0 20>
      1=<-280 0 20>
      2=<-280 0 20>
      3=<-250 0 20>
      4=<0 0 0>
      Index=&SpawnGroup()
      Num=::NewEnumerator3
     )
    )
    Position Offset Coordinate Space=::Local
    Sprite Rotation Mode=::NewEnumerator2
    Sprite Size Mode=::NewEnumerator3
    Uniform Sprite Size=&MergeHelper_FloatSwitch(
     0=&RandomRangeFloat(Minimum=25, Maximum=50)
     1=30
     2=3
     3=40
     4=0
     Index=&SpawnGroup()
     Num=::NewEnumerator3
    )
    UsePositionOffset=true
    Write Material Random=true
   )
   &PartitionParticles(Particles Per Partition=4, Partition Mode:ENiagara_PartitionMode=::NewEnumerator1)
   &SkeletalMeshLocation(
    Bone / Socket Index=&Add_Integer(A=$Particles.PartitionParticles.Partition, B=4)
    Bone / Socket Sampling Mode=::NewEnumerator5
    Interpolate Mesh Sampling=false
    MaskSpawnGroup=false
    Mesh Sampling Type=::NewEnumerator6
    Sampled Position Offset=&AddVectorToPosition(
     Position=&MaskVectorBySpawnGroup(Masked Vector=<-40 0 -20>, Passthrough Vector=<-50 0 -20>)
     Vector=$User.offset
    )
    Skeletal Mesh=$System.SkeletalMesh
   )
   &ShapeLocation(
    MaskSpawnGroup=false
    Override Seed=false
    Sphere Radius=&MergeHelper_FloatSwitch(0=30, 1=10, 2=30, 3=0, 4=0, Index=&SpawnGroup(), Num=::NewEnumerator3)
    UseOwnerScale=true
   )
   /SetVariables(Particles.VisibilityTag=&SpawnGroup()) {isSetParameters=true}
   &SubUVAnimation(
    Direct Frame Index:NiagaraFloat=&MergeHelper_FloatSwitch(
     0=&RandomRangeFloat(Minimum=0, Maximum=4)
     1=&RandomRangeFloat(Minimum=0, Maximum=4)
     2=0
     3=0
     4=0
     Index=&SpawnGroup()
     Num=::NewEnumerator3
    )
    SubUV Animation Mode:ENiagara_SubUVLookupModeV2=::NewEnumerator5
   )
   &AddVelocity(Velocity=&MergeHelper_VectorSwitch(0=<0 0 0>, 1=<-1500 0 0>, 2=<0 0 0>, 3=<0 0 0>, 4=<0 0 0>, Index=&SpawnGroup(), Num=::NewEnumerator3))
  ]
  particleUpdateScript [
   &FN_DisableTireFX(
    Mask Index Attribute:NiagaraInt32=$Particles.PartitionParticles.Partition
    Mask Index FL=2
    Mask Index FR=3
    Mask Index RL=0
    Mask Index RR=1
   )
   &ParticleState()
   &SolveForcesAndVelocity(Clamp Velocity=false, Limit Acceleration=false)
   &ScaleColor(
    Scale Alpha=&MergeHelper_FloatSwitch(0=0.5, 1=0.5, 2=1, 3=0.5, 4=0, Index=&SpawnGroup(), Num=::NewEnumerator3)
    ScaleA=true
    ScaleRGB=true
   )
   &JitterPosition(Jitter Amount:NiagaraFloat=&MergeHelper_FloatSwitch(0=0, 1=0.01, 2=1, 3=1, 4=0, Index=&SpawnGroup(), Num=::NewEnumerator3))
  ]
  rendererClasses [&NiagaraSpriteRendererProperties]
  renderers [
   {rendererClass=&NiagaraSpriteRendererProperties}
   {rendererClass=&NiagaraSpriteRendererProperties, rendererIndex=1}
   {rendererClass=&NiagaraSpriteRendererProperties, rendererIndex=2}
   {rendererClass=&NiagaraSpriteRendererProperties, rendererIndex=3}
  ]
 }
 NE_VC_DriftTrail_WheelWrap_Parent {
  bEnabled=false
  emitterUpdateScript [
   &EmitterState(
    Inactive Response=::NewEnumerator1
    Life Cycle Mode=::NewEnumerator1
    Loop Behavior=::NewEnumerator1
    Loop Duration Mode=::NewEnumerator1
    UseLoopDelay=false
   )
   &SpawnBurst_Instantaneous(Spawn Count=2, Use Spawn Probability=false, UseLoopCountLimit=false)
  ]
  particleSpawnScript [
   &VC_PaintedVariantsOverride(
    Black=<rgba 0.015 0.015 0.015 10>
    BurntSienna=<rgba 0.2396 0.057 0.0311 1>
    Cobalt=<rgba 0 0.1524 1.2 1>
    Crimson=<rgba 0.4635 0 0.0126 5>
    DefaultColor=<rgba 0.8854 0.355 0.0507 1>
    ForestGreen=<rgba 0 0.276 0.0139 5>
    Grey=<rgba 0.1514 0.1514 0.1615 5>
    Lime=<rgba 0.2599 0.4427 0 5>
    Orange=<rgba 0.6724 0.1426 0.0038 10>
    Pink=<rgba 0.7347 0.2348 0.7448 10>
    Purple=<rgba 0.1621 0.0606 0.8958 10>
    Saffron=<rgba 0.8854 0.6788 0 5>
    SkyBlue=<rgba 0.0197 0.3277 0.6 5>
    TitaniumWhite=<rgba 0.5596 0.5596 0.5625 5>
   )
   &InitializeParticle(
    Color Mode=::NewEnumerator0
    Mesh Scale Mode=::NewEnumerator4
    Sprite Rotation Mode=::NewEnumerator2
    Sprite Size Mode=::NewEnumerator3
    UsePositionOffset=false
    Write Material Random=true
   )
   &PartitionParticles()
   &SkeletalMeshLocation(
    Bone / Socket Index=&Add_Integer(A=$Particles.PartitionParticles.Partition, B=4)
    Bone / Socket Sampling Mode=::NewEnumerator5
    Interpolate Mesh Sampling=false
    MaskSpawnGroup=false
    Mesh Sampling Type=::NewEnumerator6
    Skeletal Mesh=$System.SkeletalMesh
   )
  ]
  particleUpdateScript [
   &FN_DisableTireFX(
    Mask Index Attribute:NiagaraInt32=$Particles.PartitionParticles.Partition
    Mask Index FL=2
    Mask Index FR=3
    Mask Index RL=0
    Mask Index RR=1
   )
   &VC_WheelWrapScale(Wheel Index Attribute:NiagaraInt32=&Add_Integer(A=$Particles.PartitionParticles.Partition, B=2))
   &ParticleState(Kill Particles When Lifetime Has Elapsed=false, Loop Particles Lifetime=true)
   &ScaleColor(Scale RGB=&VectorFromFloat(Value=3), ScaleA=true, ScaleRGB=true)
   &UpdateMeshOrientation(Rotation Coordinate Space=::NewEnumerator2, Rotation Vector=<0 1 0>) {enabled=false}
   &SkeletalMeshLocation(
    Bone / Socket Index=&Add_Integer(A=$Particles.PartitionParticles.Partition, B=4)
    Bone / Socket Sampling Mode=::NewEnumerator5
    MaskSpawnGroup=false
    Mesh Sampling Type=::NewEnumerator6
    Skeletal Mesh=$System.SkeletalMesh
   ) {module=SkeletalMeshLocation001}
   &SolveForcesAndVelocity(Clamp Velocity=false, Limit Acceleration=false)
  ]
  rendererClasses [&NiagaraMeshRendererProperties]
  renderers [{rendererClass=&NiagaraMeshRendererProperties, rendererIndex=0}]
 }
}
summary {
 systemName=NS_DriftTrail_Ghosthand
 emitters [
  {emitterName=NE_VC_DriftTrail_Ribbon_Parent, rendererClasses [&NiagaraRibbonRendererProperties]}
  {emitterName=NE_VC_DriftTrail_Sprite_Parent, rendererClasses [&NiagaraSpriteRendererProperties]}
  {bEnabled=false, emitterName=NE_VC_DriftTrail_WheelWrap_Parent, rendererClasses [&NiagaraMeshRendererProperties]}
 ]
 userVariables [
  User.bIsFrontendPreview {type=NiagaraBool}
  User.bIsPainted {type=NiagaraBool}
  User.ColorVariantID {type=NiagaraInt32, defaultValue=14}
  User.DTCarOffset {type=NiagaraPosition, description="", defaultValue {}}
  User.offset {type=Vector3f, description="offset from wheels(set through cosmetic template)", defaultValue {struct=Vector3f, value=<0 0 0>}}
  User.PaintedColor {type=LinearColor, description="", defaultValue {struct=LinearColor, value=<rgba 1 1 1 1>}}
  User.PaintedScalar {type=NiagaraFloat, defaultValue=1, description=""}
  User.Shine {type=NiagaraFloat, defaultValue=5, description=""}
  User.SpecMult {type=NiagaraFloat, defaultValue=0, description=""}
  User.SpecPower {type=NiagaraFloat, defaultValue=0, description=""}
  User.VehicleFlags {type=NiagaraInt32, defaultValue=806289408}
  User.Wheel_Radius_Array {
   type=NiagaraDataInterfaceArrayFloat
   defaultValue {
    struct=NiagaraExt_VariableValue_DataInterface
    value {
     dataInterface=&self:NiagaraDataInterfaceArrayFloat_1
     dataInterfaceClass=&NiagaraDataInterfaceArrayFloat
    }
   }
  }
  User.Wheel_Width_Array {
   type=NiagaraDataInterfaceArrayFloat
   defaultValue {
    struct=NiagaraExt_VariableValue_DataInterface
    value {
     dataInterface=&self:NiagaraDataInterfaceArrayFloat_0
     dataInterfaceClass=&NiagaraDataInterfaceArrayFloat
    }
   }
  }
  User.WheelTouchingArray {
   type=NiagaraDataInterfaceArrayFloat
   defaultValue {
    struct=NiagaraExt_VariableValue_DataInterface
    value {
     dataInterface=&self:NiagaraDataInterfaceArrayFloat_2
     dataInterfaceClass=&NiagaraDataInterfaceArrayFloat
    }
   }
  }
 ]
}
